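# Material function that animates the mechanical-looking eyes of the TungStan player character, with a separate aperture amount per eye.
# An optional static switch resamples the head _E texture with counter-rotated UVs inside ring masks, made by subtracting circle masks, to spin each outer eye.
# Two six-sided aperture sub-functions open and twist in mirrored directions.
# Their inverted result, clipped to inner eye circles, darkens the color by a black multiplier.
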MaterialFunction /BRCosmetics/Characters/Player/Male/Medium/Bodies/M_MED_TungStan/Materials/MaterialFunctions/MF_TungStan_Eyes
graph {nodes [
  n1 = FunctionInput() {inputName=TextureColor}
  n2 = FunctionInput() {
   bUsePreviewValueAsDefault=true
   description=Value beetween 0 and 1. How much the outer part of the eyes rotate
   inputName=ExternalRotationAmount
   inputType=FunctionInput_Scalar
  }
  n3 = FunctionInput() {bUsePreviewValueAsDefault=true, inputName=LeftApertureAmount, inputType=FunctionInput_Scalar, sortPriority=5}
  n4 = Constant2Vector() {desc=LeftEyeCirclePosition, g=0.1708, r=0.1718}
  n5 = TextureSample(UVs=MaterialFunctionCall(UVs=TextureCoordinate(), Rotation Center=@n4, "Rotation Angle (0-1)"=LinearInterpolate(B=@n2, Alpha=@n3) {constB=0}) {materialFunction=&CustomRotator}.Rotated Values) {texture=/BRCosmetics/Characters/Player/Male/Medium/Bodies/M_MED_TungStan/Textures/T_M_MED_TungStan_Head_E}
  n6 = MaterialFunctionCall!(CircleScale=Constant() {r=0.7}, CircleHardness=Constant() {r=0.98}, CenterPosition=@n4) {materialFunction=&/MF_TungStan_CircleMaskGenerator}
  n7 = MaterialFunctionCall(CircleScale=Constant() {r=0.3}, CircleHardness=Constant() {r=0.9}, CenterPosition=@n4) {materialFunction=&MF_TungStan_CircleMaskGenerator}
  n8 = Saturate(None=Subtract(A=@n6.Mask, B=@n7.Mask))
  n9 = FunctionInput() {bUsePreviewValueAsDefault=true, inputName=RightApertureAmount, inputType=FunctionInput_Scalar, sortPriority=6}
  n10 = Constant2Vector() {desc=RightEyeCirclePosition, g=0.15, r=0.363}
  n11 = MaterialFunctionCall(UVs=TextureCoordinate(), Rotation Center=@n10, "Rotation Angle (0-1)"=LinearInterpolate(B=Multiply(A=@n2) {constB=-1}, Alpha=@n9) {constB=0}) {materialFunction=&CustomRotator}
  n12 = MaterialFunctionCall(CircleScale=Constant() {r=0.7}, CircleHardness=Constant() {r=0.98}, CenterPosition=@n10) {materialFunction=&MF_TungStan_CircleMaskGenerator}
  n13 = MaterialFunctionCall(CircleScale=Constant() {r=0.3}, CircleHardness=Constant() {r=0.9}, CenterPosition=@n10) {materialFunction=&MF_TungStan_CircleMaskGenerator}
  n14 = Subtract(A=@n12.Mask, B=@n13.Mask)
  n15 = StaticSwitch(
   True=LinearInterpolate(
    A=LinearInterpolate(A=@n1, B=@n5.RGB, Alpha=@n8)
    B=TextureSample(UVs=@n11.Rotated Values) {texture=&T_M_MED_TungStan_Head_E}.RGB
    Alpha=Saturate(None=@n14)
   )
   False=@n1
   Value=FunctionInput() {bUsePreviewValueAsDefault=true, inputName=UseOuterEyesRotation, inputType=FunctionInput_StaticBool}
  )
  n16 = FunctionInput() {bUsePreviewValueAsDefault=true, inputName=EyesBlackMultiplier, inputType=FunctionInput_Scalar, previewValue=<0.02 0 0 1>}
  n17 = MaterialFunctionCall!(
   ApertureAmount=LinearInterpolate(Alpha=@n3) {constB=0.2}
   CenterPosition=Constant2Vector() {desc=LeftEyePosition, g=0.167, r=0.1718}
   ApertureRotationDirection=Constant() {r=1}
   LocalRotationAmount=LinearInterpolate(Alpha=@n3) {constB=-0.1}
   TriangleScale=Constant() {desc=TriangleScale, r=1}
  ) {materialFunction=&/MF_TungStan_6SideAperture}
  n18 = MaterialFunctionCall(
   ApertureAmount=LinearInterpolate(Alpha=@n9) {constAlpha=0, constB=0.2}
   CenterPosition=Constant2Vector() {desc=RightEyePosition, g=0.147, r=0.363}
   ApertureRotationDirection=Constant() {r=-1}
   LocalRotationAmount=LinearInterpolate(Alpha=@n9) {constAlpha=0, constB=0.1}
   TriangleScale=Constant() {desc=TriangleScale, r=1}
  ) {materialFunction=&MF_TungStan_6SideAperture}
  n19 = OneMinus(None=Saturate(None=Add(A=@n17.Result, B=@n18.Result)))
  n20 = Constant() {desc=CircleScale, r=0.28}
  n21 = Constant() {r=0.97}
  n22 = MaterialFunctionCall(CircleScale=@n20, CircleHardness=@n21, CenterPosition=Constant2Vector() {desc=CirclePosition, g=0.1708, r=0.1718}) {materialFunction=&MF_TungStan_CircleMaskGenerator}
  n23 = MaterialFunctionCall(CircleScale=@n20, CircleHardness=@n21, CenterPosition=Constant2Vector() {g=0.15, r=0.363}) {materialFunction=&MF_TungStan_CircleMaskGenerator}
  n24 = Add(A=@n22.Mask, B=@n23.Mask)
  FunctionOutput(None=LinearInterpolate(A=@n15, B=Multiply(A=@n15, B=@n16) {constB=0.02}, Alpha=LinearInterpolate(B=@n19, Alpha=@n24)) {constB=0})
 ]}
