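# Parent material for the TungStan character body, layering car-style clear coat with a separate orange-peel coat normal onto the shared 2023 character PreFX/PostFX chain.
# A multiplier adjusts roughness on metallic areas before the coat, and the PreFX normal is restored at output.
# Optional aperture eyes blink via offset, time-driven sines through the inverted eye function, with emissive scaled by time of day.
# An optional base color multiplier affects only dark areas.
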Material /BRCosmetics/Characters/Player/Male/Medium/Bodies/M_MED_TungStan/Materials/M_TungStan_Body_Parent
graph {
 properties {
  bUseMaterialAttributes=true
  bUsedWithMeshDeformer=true
  bUsedWithMorphTargets=true
  bUsedWithSkeletalMesh=true
  pixelDepthOffsetMode=PDOM_Legacy
  shadingModel=MSM_FromMaterialExpression
  usedShadingModels=Default Lit | Subsurface | Clear Coat | Subsurface Profile | Cloth
 }
 nodes [
  n1 = MaterialFunctionCall() {materialFunction=&MF_CharacterParent_2023_PreFX}
  n2 = GetMaterialAttributes(None=@n1.Result) {attributeGetTypes [#1, #2]}
  n3 = Multiply(A=@n2.Roughness, B=ScalarParameter() {defaultValue=1, parameterName=Metal Roughness Multiplier})
  n4 = Saturate(None=Power(Base=@n2.Metallic))
  n5 = MaterialFunctionCall!(
   Mask=@n1.Tech Art Mask
   Fresnel=@n1.Tech Art Mask
   In=SetMaterialAttributes!(MaterialAttributes=@n2.MaterialAttributes, Roughness=LinearInterpolate(A=@n2.Roughness, B=@n3, Alpha=@n4)) {attributeSetTypes [#1]}
  ) {materialFunction=/Game/Global/MaterialFunctions/TechArt/MF_Advanced_CarClearCoat}
  ClearCoatNormalCustomOutput!(None=MaterialFunctionCall(BaseNormal=@n5.NormalWithOrangePeel, AdditionalNormal=@n5.NormalWithOrangePeel) {materialFunction=&BlendAngleCorrectedNormals}.Result)
  n6 = GetMaterialAttributes(None=MaterialFunctionCall(Material=@n5.Result, M=@n1.M, Pre Skinned Local Position=@n1.M, Distance Blend=@n1.M, Fresnel=@n1.M) {materialFunction=&MF_CharacterParent_2023_PostFX}.Result) {attributeGetTypes [#3, #4]}
  n7 = StaticBoolParameter() {group=ApertureEyes, parameterName=UseOuterEyeRotation, sortPriority=33}
  n8 = Constant() {r=0.02}
  n9 = ScalarParameter() {defaultValue=-0.05, group=ApertureEyes, parameterName=OuterEyesRotationAmount, sortPriority=34}
  n10 = ScalarParameter() {defaultValue=0.35, group=ApertureEyes, parameterName=BlinkingThreshold_MIN, sortPriority=36}
  n11 = ScalarParameter() {defaultValue=0.6, group=ApertureEyes, parameterName=BlinkingThreshold_MAX, sortPriority=37}
  n12 = Multiply(A=Time(), B=ScalarParameter() {defaultValue=0.15, group=ApertureEyes, parameterName=BlinkingSpeed, sortPriority=38}) {constB=0.15}
  n13 = Sine(None=@n12)
  n14 = SmoothStep(Min=@n10, Max=@n11, Value=Multiply(A=Add(A=@n13)) {constB=0.5}) {constMax=0.6, constMin=0.35}
  n15 = Sine(None=Add(A=@n12, B=ScalarParameter() {defaultValue=-0.02, group=ApertureEyes, parameterName=RightEyeBlinkingOffset, sortPriority=38}) {constB=-0.02})
  n16 = SmoothStep(Min=@n10, Max=@n11, Value=Multiply(A=Add(A=@n15)) {constB=0.5}) {constMax=0.6, constMin=0.35}
  MaterialFunctionCall!(
   TextureColor=@n6.Emissive Color
   UseOuterEyesRotation=@n7
   EyesBlackMultiplier=@n8
   ExternalRotationAmount=@n9
   LeftApertureAmount=@n14
   RightApertureAmount=@n16
  ) {materialFunction=/BRCosmetics/Characters/Player/Male/Medium/Bodies/M_MED_TungStan/Materials/MaterialFunctions/MF_TungStan_Eyes}
  n17 = MaterialFunctionCall!(
   TextureColor=@n6.Emissive Color
   UseOuterEyesRotation=@n7
   EyesBlackMultiplier=@n8
   ExternalRotationAmount=@n9
   LeftApertureAmount=OneMinus(None=@n14)
   RightApertureAmount=OneMinus(None=@n16)
  ) {materialFunction=/BRCosmetics/Characters/Player/Male/Medium/Bodies/M_MED_TungStan/Materials/MaterialFunctions/MF_TungStan_EyesInverted}
  n18 = Constant() {r=0.75}
  n19 = Constant() {r=1}
  n20 = Constant() {r=0.4}
  n21 = MaterialFunctionCall(Color=@n6.Base Color) {materialFunction=&SplitComponents}
  n22 = Add(A=Add(A=@n21.R, B=@n21.R), B=@n21.B)
  n23 = LinearInterpolate(
   A=ScalarParameter() {defaultValue=1, group="BaseColorMultiplier(AffectsDarkAreasOnly)", parameterName=BaseColorMultiplier}
   Alpha=SmoothStep(Value=Saturate(None=@n22)) {constMax=0.3}
  ) {constA=1}
  SetMaterialAttributes!(
   MaterialAttributes=@n6.MaterialAttributes
   Normal=GetMaterialAttributes(None=@n1.Result) {attributeGetTypes [#5]}.Normal
   Emissive Color=Multiply(
    A=StaticSwitchParameter(True=@n17.Result, False=@n6.Emissive Color) {group=ApertureEyes, parameterName=UseApertureForEyes}
    B=MaterialFunctionCall(Value_Morning=@n18, Value_Day=@n19, Value_Evening=@n18, Value_Night=@n20) {materialFunction=&Scalar_Time_of_Day}.Result
   )
   Base Color=StaticSwitchParameter(True=Multiply(A=@n6.Base Color, B=@n23), False=@n6.Base Color) {group="BaseColorMultiplier(AffectsDarkAreasOnly)", parameterName=UseBaseColorMultiplier}
  ) {attributeSetTypes [#5, #3, #4]}
 ]
}
