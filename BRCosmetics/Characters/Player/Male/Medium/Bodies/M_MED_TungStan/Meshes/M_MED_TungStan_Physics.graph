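# Physics asset for the TungStan medium male player body cosmetic, set up for secondary motion of its dynamic waist bag.
# A capsule on the bag bone is constrained to a small sphere on its master bone, with a narrow first swing limit, a much wider second one and a small twist limit.
# An unconstrained right-thigh capsule provides a shape for the bag to collide against.
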PhysicsAsset /BRCosmetics/Characters/Player/Male/Medium/Bodies/M_MED_TungStan/Meshes/M_MED_TungStan_Physics
bodies {
 dyn_waistbag [{
   shapeType=Capsule
   length=17.5
   radius=1.5
   shapeName=dyn_waistbag_capsule
   center=<-7.6597 0.0001 2.5965>
   rotation=<rot -0.0003 -0.0005 90>
  }]
 dyn_waistbag_MSTR [{shapeType=Sphere, radius=1, shapeName=dyn_waistbag_MSTR_sphere}]
 thigh_r [{shapeType=Capsule, length=22.7, radius=7.3, shapeName=thigh_r_capsule, center=<9.4596 -0.0001 -0.3606>, rotation=<rot -90 180 179.9996>}]
}
constraints [{bone1Name=dyn_waistbag, bone2Name=dyn_waistbag_MSTR, swing1LimitDegrees=5, swing2LimitDegrees=60, twistLimitDegrees=10}]
