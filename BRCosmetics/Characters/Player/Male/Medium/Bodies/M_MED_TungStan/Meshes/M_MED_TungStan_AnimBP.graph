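# Anim Blueprint for the TungStan male body cosmetic, adding secondary motion to clothing and accessories.
# The parent mesh's pose feeds Control Rig, waist-bag corrections, RigidBody, AnimDynamics for sleeves, shorts and zippers, then a crouch belt offset.
# Each update refreshes gravity override, movement-driven alphas and three-region material WPO cloth parameters.
# WPO defaults are cached then zeroed on customization load, restored after a skydiving delay and zeroed when skydiving ends.
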AnimBlueprint /BRCosmetics/Characters/Player/Male/Medium/Bodies/M_MED_TungStan/Meshes/M_MED_TungStan_AnimBP
target_skeleton=&/M_MED_TungStan_Skeleton
graphs [
 AnimGraph {nodes [
   "Transform (Modify) Bone - Bone: dyn_waistbag_MSTR"!(
    ComponentPose
    bAlphaBoolEnabled:bool=Get bIsJumpingOrFalling() -> (bIsJumpingOrFalling:bool)
    Translation="1,-3,1"
    Rotation="-15,0,-15"
    Scale="1,1,1"
   ) -> (Pose) {
    alphaInputType=Bool
    rotationMode=BMM_Additive
    rotationSpace=BCS_ParentBoneSpace
    translationMode=BMM_Additive
    translationSpace=BCS_ParentBoneSpace
    alphaBoolBlend {blendInTime=0.3, blendOption=Linear}
    boneToModify {boneName=dyn_waistbag_MSTR}
    rotation {pitch=-15, roll=-15}
    translation=<1 -3 1>
   }
   n1 = Control Rig(Source=Copy Pose From Mesh() {bUseAttachedParent=true}) {
    alphaBoolBlend {blendOption=Linear}
    controlRigAssetReference {blueprintRigClass=&/M_MED_TungStan_ControlRig.M_MED_TungStan_ControlRig_C}
   }
   n2 = "Transform (Modify) Bone - Bone: dyn_waistbag_MSTR"!(
    ComponentPose=Local To Component(LocalPose=@n1)
    Alpha:float=Get WaistBagAlpha() -> (WaistBagAlpha:float)
    Translation="1,-3,2"
    Rotation="-15,0,-30"
    Scale="1,1,1"
   ) -> (Pose) {
    rotationMode=BMM_Additive
    rotationSpace=BCS_ParentBoneSpace
    translationMode=BMM_Additive
    translationSpace=BCS_ParentBoneSpace
    alphaBoolBlend {blendOption=Linear}
    alphaScaleBiasClamp {bInterpResult=true}
    boneToModify {boneName=dyn_waistbag_MSTR}
    rotation {pitch=-15, roll=-30}
    translation=<1 -3 2>
   }
   n3 = "Transform (Modify) Bone - Bone: dyn_waistbag_MSTR"!(
    ComponentPose=@n2
    bAlphaBoolEnabled:bool=Get bIsCrouchedSideMoving() -> (bIsCrouchedSideMoving:bool)
    Translation="0,0,0.5"
    Rotation="0,0,0"
    Scale="1,1,1"
   ) -> (Pose) {
    alphaInputType=Bool
    rotationMode=BMM_Additive
    rotationSpace=BCS_ParentBoneSpace
    translationMode=BMM_Additive
    translationSpace=BCS_ParentBoneSpace
    alphaBoolBlend {blendInTime=0.5, blendOption=Linear, blendOutTime=0.2}
    boneToModify {boneName=dyn_waistbag_MSTR}
    translation {z=0.5}
   }
   n4 = Look At - Bone: dyn_waistbag_MSTR(ComponentPose=@n3, Alpha=Get WaistBagLookAtCalfAlpha() -> (WaistBagLookAtCalfAlpha:double)) {
    alphaBoolBlend {blendOption=Linear}
    boneToModify {boneName=dyn_waistbag_MSTR}
    lookAtLocation {x=0}
    lookAtTarget {boneReference {boneName=calf_r}}
    lookAt_Axis {axis=<-1 0>}
   }
   n5 = RigidBody!(ComponentPose=@n4, Alpha, OverrideWorldGravity:Vector=Get Gravity Override()) -> (Pose) {
    bOverrideWorldGravity=true
    overridePhysicsAsset=&/M_MED_TungStan_Physics
    simulationSpace=BaseBoneSpace
    alphaBoolBlend {blendOption=Linear}
    baseBoneRef {boneName=pelvis}
    componentAppliedLinearAccClamp=<3000 3000 3000>
    componentLinearAccScale=<0.3 0.3 0.3>
    componentLinearVelScale=<0.3 0.3 0.3>
   }
   n6 = Get Gravity Override()
   n7 = Get AnimDynamicsSleevesAlpha() -> (AnimDynamicsSleevesAlpha:double)
   n8 = Anim Dynamics - Bone: dyn_sleeve_l!(ComponentPose=@n5, Alpha=@n7, GravityOverride:Vector=@n6) -> (Pose) {
    angularDampingOverride=0.8
    bOverrideAngularDamping=true
    bUseGravityOverride=true
    bUsePlanarLimit=false
    simulationSpace=BoneRelative
    alphaBoolBlend {blendOption=Linear}
    alphaScaleBiasClamp {bInterpResult=true}
    boundBone {boneName=dyn_sleeve_l}
    componentAppliedLinearAccClamp=<1000 1000 1000>
    componentLinearAccScale=<1 1 1>
    componentLinearVelScale=<1 1 1>
    gravityOverride {z=-980}
    physicsBodyDefinitions [{
      boundBone {boneName=dyn_sleeve_l}
      boxExtents=<5 20 20>
      constraintSetup {angularLimitsMax=<0 7 10>, angularLimitsMin=<0 -7 -10>, angularTarget=<0 0 0>}
      localJointOffset=<10 0 0>
     }]
    relativeSpaceBone {boneName=clavicle_l}
   }
   n9 = Anim Dynamics - Bone: dyn_sleeve_r!(ComponentPose=@n8, Alpha=@n7, GravityOverride:Vector=@n6) -> (Pose) {
    angularDampingOverride=0.8
    bOverrideAngularDamping=true
    bUseGravityOverride=true
    bUsePlanarLimit=false
    simulationSpace=BoneRelative
    alphaBoolBlend {blendOption=Linear}
    alphaScaleBiasClamp {bInterpResult=true}
    boundBone {boneName=dyn_sleeve_r}
    componentAppliedLinearAccClamp=<1000 1000 1000>
    componentLinearAccScale=<1 1 1>
    componentLinearVelScale=<1 1 1>
    gravityOverride {z=-980}
    physicsBodyDefinitions [{
      boundBone {boneName=dyn_sleeve_r}
      boxExtents=<5 20 20>
      constraintSetup {angularLimitsMax=<0 7 10>, angularLimitsMin=<0 -7 -10>, angularTarget=<0 0 0>}
      localJointOffset=<-10 0 0>
     }]
    relativeSpaceBone {boneName=clavicle_r}
   }
   n10 = Get Gravity Override()
   n11 = Get AnimDynamicsShortsAlpha() -> (AnimDynamicsShortsAlpha:double)
   n12 = Anim Dynamics - Bone: dyn_short_l!(ComponentPose=@n9, Alpha=@n11, GravityOverride:Vector=@n10) -> (Pose) {
    angularDampingOverride=0.8
    bOverrideAngularDamping=true
    bUseGravityOverride=true
    bUsePlanarLimit=false
    simulationSpace=BoneRelative
    alphaBoolBlend {blendOption=Linear}
    alphaScaleBiasClamp {bInterpResult=true}
    boundBone {boneName=dyn_short_l}
    componentAppliedLinearAccClamp=<2500 2500 2500>
    componentLinearAccScale=<0.5 0.5 0.5>
    componentLinearVelScale=<0.5 0.5 0.5>
    gravityOverride {z=-980}
    physicsBodyDefinitions [{
      boundBone {boneName=dyn_short_l}
      boxExtents=<7 5 5>
      constraintSetup {angularLimitsMax=<0 5 10>, angularLimitsMin=<0 -5 -10>, angularTarget=<0 0 0>}
      localJointOffset=<-10 0 0>
     }]
    relativeSpaceBone {boneName=thigh_l}
   }
   n13 = Anim Dynamics - Bone: dyn_short_r!(ComponentPose=@n12, Alpha=@n11, GravityOverride:Vector=@n10) -> (Pose) {
    angularDampingOverride=0.8
    bOverrideAngularDamping=true
    bUseGravityOverride=true
    bUsePlanarLimit=false
    simulationSpace=BoneRelative
    alphaBoolBlend {blendOption=Linear}
    alphaScaleBiasClamp {bInterpResult=true}
    boundBone {boneName=dyn_short_r}
    componentAppliedLinearAccClamp=<2500 2500 2500>
    componentLinearAccScale=<0.5 0.5 0.5>
    componentLinearVelScale=<0.5 0.5 0.5>
    gravityOverride {z=-980}
    physicsBodyDefinitions [{
      boundBone {boneName=dyn_short_r}
      boxExtents=<7 5 5>
      constraintSetup {angularLimitsMax=<0 5 10>, angularLimitsMin=<0 -5 -10>, angularTarget=<0 0 0>}
      localJointOffset=<10 0 0>
     }]
    relativeSpaceBone {boneName=thigh_r}
   }
   n14 = Get Gravity Override()
   n15 = Anim Dynamics - Bone: dyn_waistbag_zipper!(ComponentPose=@n13, Alpha=1, GravityOverride:Vector=@n14) -> (Pose) {
    angularDampingOverride=0.95
    bOverrideAngularDamping=true
    bUseGravityOverride=true
    bUsePlanarLimit=false
    simulationSpace=BoneRelative
    alphaBoolBlend {blendOption=Linear}
    boundBone {boneName=dyn_waistbag_zipper}
    componentAppliedLinearAccClamp=<1000 1000 1000>
    componentLinearAccScale=<0.5 0.5 0.5>
    componentLinearVelScale=<0.5 0.5 0.5>
    gravityOverride {z=-980}
    physicsBodyDefinitions [{
      boundBone {boneName=dyn_waistbag_zipper}
      boxExtents=<8 10 10>
      constraintSetup {angularLimitsMax=<0 60 90>, angularLimitsMin=<0 -30 -15>, angularTarget=<0 0 0>}
      localJointOffset=<2.5 0 0>
     }]
    relativeSpaceBone {boneName=root}
   }
   n16 = Anim Dynamics - Bone: dyn_chestbag_zipper!(ComponentPose=@n15, Alpha=1, GravityOverride:Vector=@n14) -> (Pose) {
    angularDampingOverride=0.85
    bOverrideAngularDamping=true
    bUseGravityOverride=true
    bUsePlanarLimit=false
    simulationSpace=BoneRelative
    alphaBoolBlend {blendOption=Linear}
    boundBone {boneName=dyn_chestbag_zipper}
    componentAppliedLinearAccClamp=<2500 2500 2500>
    componentLinearAccScale=<0.5 0.5 0.5>
    componentLinearVelScale=<0.5 0.5 0.5>
    gravityOverride {z=-980}
    physicsBodyDefinitions [{
      boundBone {boneName=dyn_chestbag_zipper}
      boxExtents=<8 15 15>
      constraintSetup {angularLimitsMax=<0 15 90>, angularLimitsMin=<0 -15 -5>, angularTarget=<0 0 0>}
      localJointOffset=<2.5 0 0>
     }]
    relativeSpaceBone {boneName=root}
   }
   n17 = "Transform (Modify) Bone - Bone: dyn_belt"!(ComponentPose=@n16, bAlphaBoolEnabled:bool=Get bIsCrouching(), Translation="15,0,0", Rotation="0,0,0", Scale="1,1,1") -> (Pose) {
    alphaInputType=Bool
    rotationMode=BMM_Additive
    rotationSpace=BCS_BoneSpace
    translationMode=BMM_Additive
    translationSpace=BCS_ParentBoneSpace
    alphaBoolBlend {blendOption=Linear}
    boneToModify {boneName=dyn_belt}
    translation {x=15}
   }
   Output Pose(Result=Component To Local(ComponentPose=@n17)) {layerGroup=None}
  ]}
 DisableWPO {nodes [
   DisableWPO() {is=title K2Node_FunctionEntry}
   > n1 = For Loop(FirstIndex=1, LastIndex=Length(TargetArray:Array of Vector 4 Structures=Get WPODefaults() -> (WPODefaults:Array of Vector 4 Structures)))
   > Set L_RegionNum(L_RegionNum:int=@n1) -> (Output_Get:int=0)
   n2 = "To String (Integer)"(InInt=Get L_RegionNum() -> (L_RegionNum:int=0))
   n3 = Get Mat Instance Dynamic()
   > SetScalarParameterValue(self=@n3, ParameterName=String To Name(InString=Append!(A=ClothSim_R, B=@n2, C:String=_DragStrength) -> (ReturnValue)), Value=0)
   n4 = "To String (Integer)"(InInt=Get L_RegionNum() -> (L_RegionNum=0))
   > SetScalarParameterValue(self=@n3, ParameterName=String To Name(InString=Append(A=ClothSim_R, B=@n4, C=_TurbulenceStrength)), Value=0)
   n5 = "To String (Integer)"(InInt=Get L_RegionNum() -> (L_RegionNum=0))
   > SetScalarParameterValue(self=@n3, ParameterName=String To Name(InString=Append(A=ClothSim_R, B=@n5, C=_BoneRadius)), Value=0)
   n6 = "To String (Integer)"(InInt=Get L_RegionNum() -> (L_RegionNum=0))
   > SetScalarParameterValue(self=@n3, ParameterName=String To Name(InString=Append(A=ClothSim_R, B=@n6, C=_BoneThreshold)), Value=0)
  ]}
 EnableWPO {nodes [
   EnableWPO() {is=title K2Node_FunctionEntry}
   n1 = Get WPODefaults()
   > n2 = For Loop(FirstIndex=1, LastIndex=Length(TargetArray=@n1))
   > n3 = Set L_RegionNum(L_RegionNum=@n2) -> (Output_Get=0)
   > n4 = Set L_WPOSettings!(
    execute
    L_WPOSettings:Vector 4 Structure="Get (a copy)"(Array:Array of Vector 4 Structures=@n1, Dimension 1=int - int(A=@n3)) -> (Output:Vector 4 Structure)
   ) -> (then, Output_Get_X:double, Output_Get_Y:double, Output_Get_Z:double, Output_Get_W:double)
   n5 = "To String (Integer)"(InInt=Get L_RegionNum() -> (L_RegionNum=0))
   n6 = Get Mat Instance Dynamic()
   > SetScalarParameterValue(self=@n6, ParameterName=String To Name(InString=Append(A=ClothSim_R, B=@n5, C=_DragStrength)), Value=@n4)
   n7 = "To String (Integer)"(InInt=Get L_RegionNum() -> (L_RegionNum=0))
   > SetScalarParameterValue(self=@n6, ParameterName=String To Name(InString=Append(A=ClothSim_R, B=@n7, C=_TurbulenceStrength)), Value=@n4.Output_Get_Y)
   n8 = "To String (Integer)"(InInt=Get L_RegionNum() -> (L_RegionNum=0))
   > SetScalarParameterValue(self=@n6, ParameterName=String To Name(InString=Append(A=ClothSim_R, B=@n8, C=_BoneRadius)), Value=@n4.Output_Get_Z)
   n9 = "To String (Integer)"(InInt=Get L_RegionNum() -> (L_RegionNum=0))
   > SetScalarParameterValue(self=@n6, ParameterName=String To Name(InString=Append(A=ClothSim_R, B=@n9, C=_BoneThreshold)), Value=@n4.Output_Get_W)
  ]}
 EventGraph {nodes [
   n1 = Event BlueprintUpdateAnimation()
   > Set Delta Time X(Delta Time X=@n1.DeltaTimeX)
   > K2Node_Composite_0 = Gravitty Override!(execute) -> (then:Exec) {is=title K2Node_Composite}
   > K2Node_Composite_2 = Varibles!(execute) -> (then:Exec) {is=title K2Node_Composite}
   > K2Node_Composite_1 = WPO!(execute) -> (then:Exec) {is=title K2Node_Composite}
   Event BlueprintInitializeAnimation()
   > Set Skel Mesh(Skel Mesh=GetOwnerSkeletalMeshComponent())
   n2 = TryGetPawnOwner()
   > Is Valid(InputObject=@n2)
   > Set PlayerPawn(PlayerPawn:Pawn Object Reference=@n2) -> (Output_Get:Pawn Object Reference)
   Event OnBeginSkydiving()
   > RetriggerableDelay(Duration=3)
   > EnableWPO() {is=title K2Node_CallFunction}
   Event OnCharacterCustomizationLoaded()
   n3 = GetPartSkeletalMeshComponent()
   > n4 = Get Dynamic Material Instance(SkeletalMeshComponent=@n3)
   > Set Mat Instance Dynamic(Mat Instance Dynamic=@n4)
   > n5 = GetOrCreateDynamicOverlayMaterial(MeshComponent=@n3, MaterialName=None)
   > Set Overlay Mat Dynamic(Overlay Mat Dynamic:Material Instance Dynamic Object Reference=@n5) -> (Output_Get:Material Instance Dynamic Object Reference)
   > Do Once()
   > SaveWPO!(execute, self, Regions:int=3) -> (then)
   > DisableWPO() {is=title K2Node_CallFunction}
   Event OnEndSkydiving()
   > DisableWPO()
  ]}
 Gravitty Override {
  path=EventGraph.K2Node_Composite_0.Gravitty Override
  nodes [
   Inputs()
   > n1 = Gravity Override State Function DEPRECATED(CharacterPartAnimInstance=Self-Reference(), GameTimeInSeconds=GetGameTimeInSeconds(), GravityOverrideParams=Get Gravity Override Params())
   > Set Gravity Override(Gravity Override=@n1)
   > Outputs!(then:Exec)
  ]
 }
 WPO {
  path=EventGraph.K2Node_Composite_1.WPO
  nodes [
   Inputs()
   > Is Valid(InputObject=Get PlayerPawn() -> (PlayerPawn:Pawn Object Reference))
   > Set WPO Globals Function(
    Material=Get Mat Instance Dynamic()
    Overlay Material=Get Overlay Mat Dynamic() -> (Overlay Mat Dynamic:Material Instance Dynamic Object Reference)
    Skeletal Mesh=Get Skel Mesh()
    "Local Noise Transform  Location"="Get Local Noise Transform  Location"() -> ("Local Noise Transform  Location":Vector)
    Local Noise Offset=Get Local Noise Offset()
   )
   > Set WPO Regional Function!(
    execute
    Material:Material Instance Dynamic Object Reference=Get Mat Instance Dynamic()
    Overlay Material:Material Instance Dynamic Object Reference=Get Overlay Mat Dynamic()
    Skeletal Mesh:Skeletal Mesh Component Object Reference=Get Skel Mesh()
    Velocity Transform:Name=pelvis
    Velocity Interp Speed:double=Get Velocity Interp Speed()
    Collision Transform:Name=spine_03
    Collision Transform Axis:EAxis Enum
    Invert Collision Transform Axis:bool=false
    Delta Time X:double=Get Delta Time X()
    Previous Velocity Transform Location:"Vector (by ref)"=Get R1 Velocity Transform Location()
    Previous Velocity Transform Velocity:"Vector (by ref)"=Get R1 Velocity Transform Velocity()
    Region:int=1
   ) -> (then)
   > Set WPO Regional Function(
    Material=Get Mat Instance Dynamic()
    Overlay Material=Get Overlay Mat Dynamic()
    Skeletal Mesh=Get Skel Mesh()
    Velocity Transform=lowearm_r
    Velocity Interp Speed=Get Velocity Interp Speed()
    Collision Transform=upperarm_r
    Invert Collision Transform Axis=false
    Delta Time X=Get Delta Time X()
    Previous Velocity Transform Location=Get R2 Velocity Transform Location()
    Previous Velocity Transform Velocity=Get R2 Velocity Transform Velocity()
    Region=2
   )
   > Set WPO Regional Function(
    Material=Get Mat Instance Dynamic()
    Overlay Material=Get Overlay Mat Dynamic()
    Skeletal Mesh=Get Skel Mesh()
    Velocity Transform=calf_r
    Velocity Interp Speed=Get Velocity Interp Speed()
    Collision Transform=pelvis
    Invert Collision Transform Axis=false
    Delta Time X=Get Delta Time X()
    Previous Velocity Transform Location=Get R3 Velocity Transform Location() -> (R3 Velocity Transform Location:Vector)
    Previous Velocity Transform Velocity=Get R3 Velocity Transform Velocity() -> (R3 Velocity Transform Velocity:Vector)
    Region=3
   )
   > SetScalarParameterValue(self=Get Mat Instance Dynamic(), ParameterName=ClothSim_WindStrength, Value=SelectFloat(A=1, B=0, bPickA=Get bIsSkydiveFloating()))
   > Outputs()
  ]
 }
 Varibles {
  path=EventGraph.K2Node_Composite_2.Varibles
  nodes [
   Inputs()
   n1 = GetAlphaFromReferenceTransformDelta(SkeletalMeshComponent=Get Skel Mesh(), BoneName=thigh_r, BaseName=pelvis, InputRange=(X=-17,Y=-90), TwistAxis="1,0,0")
   > Set WaistBagAlpha(WaistBagAlpha:float=SelectFloat(A=@n1, B="Clamp (Float)"(Value=@n1, Min=0, Max=0.2), bPickA=Get bIsCrouching())) -> (Output_Get:float)
   > Set AnimDynamicsSleevesAlpha(AnimDynamicsSleevesAlpha:double=SelectFloat(A=0.2, B=1, bPickA=Get bIsFalling())) -> (Output_Get:double)
   n2 = Get Delta Time X()
   > Set AnimDynamicsShortsAlpha(AnimDynamicsShortsAlpha:double=SelectFloat(A=0.6, B=MapRangeClamped(Value=@n2, InRangeA=0.016, InRangeB=0.04, OutRangeA=1, OutRangeB=0), bPickA=Get bIsFalling())) -> (Output_Get:double)
   n3 = Get PawnMovementDirectionAzimuth() -> (PawnMovementDirectionAzimuth:float)
   n4 = AND Boolean(A="float > float"(A=@n3, B=30), B="float < float"(A=@n3, B=135))
   n5 = AND Boolean(A="float > float"(A=@n3, B=-135), B="float < float"(A=@n3, B=-30))
   > n6 = Set bIsCrouchedSideMoving(bIsCrouchedSideMoving:bool=AND Boolean!(A=OR Boolean(A=@n4, B=@n5), B=Get bIsCrouching(), C:bool=Get bIsMoving2D()) -> (ReturnValue)) -> (Output_Get:bool)
   > Set WaistBagLookAtCalfAlpha(WaistBagLookAtCalfAlpha:double=SelectFloat(A=0.3, B=0, bPickA=@n6)) -> (Output_Get:double)
   > Set bIsJumpingOrFalling(bIsJumpingOrFalling:bool=OR Boolean(A=Get bIsJumping(), B=Get bIsFalling())) -> (Output_Get:bool)
   > Outputs()
  ]
 }
 SaveWPO {nodes [
   n1 = SaveWPO!() -> (then, Regions:int=1) {is=title K2Node_FunctionEntry}
   > n2 = For Loop(FirstIndex=1, LastIndex=@n1)
   > Set L_RegionNum(L_RegionNum=@n2) -> (Output_Get=0)
   n3 = "To String (Integer)"(InInt=Get L_RegionNum() -> (L_RegionNum=0))
   > n4 = Get Scalar Parameter Value!(
    execute
    self:Material Instance Dynamic Object Reference=Get Mat Instance Dynamic()
    ParameterName:Name=String To Name(InString=Append(A=ClothSim_R, B=@n3, C=_DragStrength))
   ) -> (then, ReturnValue:float) {is=title K2Node_CallFunction}
   n5 = "To String (Integer)"(InInt=Get L_RegionNum() -> (L_RegionNum=0))
   > n6 = Get Scalar Parameter Value(self=Get Mat Instance Dynamic(), ParameterName=String To Name(InString=Append(A=ClothSim_R, B=@n5, C=_TurbulenceStrength)))
   n7 = "To String (Integer)"(InInt=Get L_RegionNum() -> (L_RegionNum=0))
   > n8 = Get Scalar Parameter Value(self=Get Mat Instance Dynamic(), ParameterName=String To Name(InString=Append(A=ClothSim_R, B=@n7, C=_BoneRadius)))
   n9 = "To String (Integer)"(InInt=Get L_RegionNum() -> (L_RegionNum=0))
   > n10 = Get Scalar Parameter Value(self=Get Mat Instance Dynamic(), ParameterName=String To Name(InString=Append(A=ClothSim_R, B=@n9, C=_BoneThreshold)))
   > Add(
    TargetArray:Array of Vector 4 Structures=Get WPODefaults()
    NewItem:"Vector 4 Structure (by ref)"=MakeVector4!(X:double=@n4, Y:double=@n6, Z:double=@n8, W:double=@n10) -> (ReturnValue:Vector 4 Structure)
   ) {is=title K2Node_CallArrayFunction}
  ]}
]
variables [
 AnimDynamicsShortsAlpha {type=number}
 AnimDynamicsSleevesAlpha {type=number}
 bIsCrouchedSideMoving {type=boolean}
 bIsJumpingOrFalling {type=boolean}
 Delta Time X {type=number}
 Gravity Override {type=Vector}
 Gravity Override Params {type=GravityOverrideParamsStruct}
 Local Noise Offset {type=Vector}
 "Local Noise Transform  Location" {type=Vector}
 Mat Instance Dynamic {type=MaterialInstanceDynamic}
 Overlay Mat Dynamic {type=MaterialInstanceDynamic}
 PlayerPawn {type=Pawn, default=null}
 R1 Velocity Transform Location {type=Vector}
 R1 Velocity Transform Velocity {type=Vector}
 R2 Velocity Transform Location {type=Vector}
 R2 Velocity Transform Velocity {type=Vector}
 R3 Velocity Transform Location {type=Vector}
 R3 Velocity Transform Velocity {type=Vector}
 Skel Mesh {type=SkeletalMeshComponent}
 Velocity Interp Speed {type=number, default=10}
 WaistBagAlpha {type=number}
 WaistBagLookAtCalfAlpha {type=number}
 WPODefaults {type="array<Vector4>", default []}
 XAlpha {type=number}
 YAlpha {type=number}
]
